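# Animation Blueprint for the M_MED_Headband_S backpack that adjusts the pack's pose during skydiving.
# Each update sets skydiveNoParachute to 1 when skydiving with the parachute closed, otherwise 0.
# That value drives the alpha of an additive bone-space offset and roll on the pack bone over the reference pose.
# A spine_05 yaw mapping is computed in the graph but feeds nothing, and Spine05_Rot_Alpha is unused.
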AnimBlueprint /BRCosmetics/Accessories/FORT_Backpacks/Backpack_M_MED_Headband_S/Meshes/M_MED_Headband_S_Pack_AnimBP
target_skeleton=&/M_MED_Headband_S_Pack_Skeleton
graphs [
 AnimGraph {nodes [
   n1 = "Transform (Modify) Bone - Bone: pack"(
    ComponentPose=Local To Component(LocalPose=Local Space Ref Pose())
    Alpha=Get skydiveNoParachute() -> (skydiveNoParachute:double)
    Translation="0,0.5,1.5"
    Rotation="0.000,0.000,10"
   ) {
    rotationMode=BMM_Additive
    rotationSpace=BCS_BoneSpace
    translationMode=BMM_Additive
    translationSpace=BCS_BoneSpace
    alphaBoolBlend {blendOption=Linear}
    boneToModify {boneName=pack}
    rotation {roll=10}
    translation {y=0.5, z=1.5}
   }
   Output Pose(Result=Component To Local(ComponentPose=@n1)) {layerGroup=None}
  ]}
 EventGraph {nodes [
   Event BlueprintUpdateAnimation()
   > n1 = Sequence()
   n2 = NOT Boolean(A=Get bIsParachuteOpened())
   n3 = Get bIsSkyDiving()
   Set skydiveNoParachute(execute=@n1.then_1, skydiveNoParachute:double="To Float (Boolean)"(InBool=AND Boolean(A=@n3, B=@n2))) -> (Output_Get:double)
   TryGetPawnOwner()
   n4 = GetDeltaTransformFromRefPose(self=GetOwnerSkeletalMeshComponent(), BoneName=spine_05, BaseName=spine_04)
   MapRangeClamped(Value=BreakRotator(InRot=BreakTransform(InTransform=@n4).Rotation).Yaw, InRangeA=-1, InRangeB=-12)
  ]}
]
variables [
 skydiveNoParachute {type=number}
 Spine05_Rot_Alpha {type=number}
]
